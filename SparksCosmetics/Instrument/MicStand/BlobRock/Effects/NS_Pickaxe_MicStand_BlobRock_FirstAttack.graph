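# Niagara effect for the first attack of the BlobRock MicStand pickaxe, placing short air-swipe bursts on the User skeletal mesh.
# Air_cone spawns a blue mesh at a sampled socket, and Airflow_Stick a gray, drifting mesh at a sampled bone, each after a brief delay.
# Both use curves to stretch the meshes, drive dynamic material parameters and fade out.
# A Droplets sprite emitter with socket-following spray is present but disabled.
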NiagaraSystem /SparksCosmetics/Instrument/MicStand/BlobRock/Effects/NS_Pickaxe_MicStand_BlobRock_FirstAttack
dependencies {
 usedDataInterfaces [&NiagaraDataInterfaceCurve, &NiagaraDataInterfaceColorCurve, &NiagaraDataInterfaceVector2DCurve]
 usedDynamicInputs []
 usedModules [&SystemState]
}
emitters {
 Air_cone {
  emitterUpdateScript [
   &EmitterState(Life Cycle Mode=::NewEnumerator1, Loop Behavior=::NewEnumerator1, Loop Delay=0.08, Loop Duration=0.5, UseLoopDelay=true)
   &SpawnBurst_Instantaneous(Use Spawn Probability=false, UseLoopCountLimit=false)
  ]
  particleSpawnScript [
   &InitializeParticle(
    Color=<rgba 0.1089 0.1429 1.1 1>
    Lifetime=0.2
    Mesh Scale=&Multiply_VectorByFloat(Vector=<0.42 0.14 0.14>, Float=1)
    Mesh Scale Mode=::NewEnumerator4
    UsePositionOffset=false
    Write Material Random=true
   )
   &InitialMeshOrientation(Mesh Orientation Mode=::NewEnumerator2, Rotation=<0 0.25 0>, Use Rotation Vector=true)
   &SkeletalMeshLocation(
    MaskSpawnGroup=false
    Mesh Sampling Type=::NewEnumerator5
    Sampled Position Offset=<0 0 -15>
    Skeletal Mesh=$User.Skeletal Mesh
    Socket Sampling Mode=::NewEnumerator5
   )
  ]
  particleUpdateScript [
   &ParticleState()
   &ScaleMeshSize(Scale Factor=&MakeVector(
     X=&FloatFromCurve(
      FloatCurve=DataInterface() {propertyValues {Curve {keys [{arriveTangent=0.286, interpMode=RCIM_Cubic, leaveTangent=0.286, tangentMode=RCTM_User, time=0, value=1}, <at 1 1.2>]}}}
      Scale Curve=0.65
     )
     Y=&FloatFromCurve(
      FloatCurve=DataInterface() {propertyValues {Curve {keys [{arriveTangent=3.1084, interpMode=RCIM_Cubic, leaveTangent=3.1084, tangentMode=RCTM_User, time=0, value=0.5}, <at 0.3 1>]}}}
      Scale Curve=2
     )
     Z=&FloatFromCurve(
      FloatCurve=DataInterface() {propertyValues {Curve {keys [{arriveTangent=3.1084, interpMode=RCIM_Cubic, leaveTangent=3.1084, tangentMode=RCTM_User, time=0, value=0.5}, <at 0.3 1>]}}}
      Scale Curve=2
     )
    ))
   &DynamicMaterialParameters(
    Index 0 Param 1=0.25
    Index 0 Param 2=0.5
    Index 0 Param 3=1
    Index 0 Param 4=&FloatFromCurve(FloatCurve=DataInterface() {propertyValues {Curve {keys [{arriveTangent=-1.3436, interpMode=RCIM_Cubic, leaveTangent=-1.3436, tangentMode=RCTM_User, time=0, value=0.75}, <at 1 0>]}}})
    Index 3 Param 2=1
    Module.Param0WriteEnabled=true
    Module.Param0WriteEnabled3=true
    Module.Param1WriteEnabled=true
    Module.Param1WriteEnabled3=true
    Module.Param2WriteEnabled=true
    Module.Param2WriteEnabled3=true
    Module.Param3WriteEnabled=true
    Module.Param3WriteEnabled3=true
    Write Parameter Index 3=true
   ) {module=DynamicMaterialParameters001}
   &ScaleColor(
    Scale Alpha=&FloatFromCurve(FloatCurve=DataInterface() {propertyValues {Curve {keys [<at 0 1>,<at 1 0>]}}}, Scale Curve=0.5)
    Scale RGB=&MakeVectorFromLinearColorRGB(Color=&ColorFromCurve(ColorCurve=DataInterface() {propertyValues {
        AlphaCurve {keys []}
        BlueCurve {keys [<at 0 5 RCIM_Cubic>,<at 0.4076 0.5 RCIM_Cubic>]}
        GreenCurve {keys [<at 0 5 RCIM_Cubic>,<at 0.4076 0.4844 RCIM_Cubic>]}
        RedCurve {keys [<at 0 5 RCIM_Cubic>,<at 0.4076 0.4922 RCIM_Cubic>]}
       }}))
    ScaleA=true
    ScaleRGB=true
   )
   &FN_Mesh_Attributes(
    Color=&MakeLinearColorFromVectorAndFloat(
     "Vector (RGB)"=&Multiply_Vector(A=&MakeVectorFromLinearColorRGB(Color=<rgba 0.0093 0 0.599 1>), B=<1 1 1>)
     "Float (Alpha)"=&FloatFromCurve(
      FloatCurve=DataInterface() {propertyValues {Curve {keys [{arriveTangent=-0.2188, interpMode=RCIM_Cubic, leaveTangent=-0.2188, tangentMode=RCTM_User, time=0, value=1}, <at 1 0>]}}}
      Scale Curve=&Multiply_Float(A=2, B=1)
     )
    )
    Create Color Attribute=true
    Create Scale Attribute=true
    Scale=&Multiply_Vector(
     A=<0.12 0.25 0.25>
     B=&MakeVector(
      X=&FloatFromCurve(FloatCurve=DataInterface() {propertyValues {Curve {keys [{arriveTangent=0.3763, interpMode=RCIM_Cubic, leaveTangent=0.3763, tangentMode=RCTM_User, time=0, value=1}, <at 1 1.2>]}}})
      Y=&FloatFromCurve(FloatCurve=DataInterface() {propertyValues {Curve {keys [{arriveTangent=3.1084, interpMode=RCIM_Cubic, leaveTangent=3.1084, tangentMode=RCTM_User, time=0, value=0.5}, <at 0.3 1>]}}})
      Z=&FloatFromCurve(FloatCurve=DataInterface() {propertyValues {Curve {keys [{arriveTangent=3.1084, interpMode=RCIM_Cubic, leaveTangent=3.1084, tangentMode=RCTM_User, time=0, value=0.5}, <at 0.3 1>]}}})
     )
    )
    Write to Module Namespace=true
   ) {module=FN_Mesh_Attributes001}
   /SetVariables(Particles.Dy3:Vector4f=&MakeVector4(X=0, Y=1, Z=0, W=1)) {isSetParameters=true}
  ]
  rendererClasses [&NiagaraMeshRendererProperties]
  renderers [
   {rendererClass=&NiagaraMeshRendererProperties, rendererIndex=0}
   {rendererClass=&NiagaraMeshRendererProperties, rendererIndex=1}
   {rendererClass=&NiagaraMeshRendererProperties, rendererIndex=2}
  ]
 }
 Airflow_Stick {
  emitterUpdateScript [
   &EmitterState(Life Cycle Mode=::NewEnumerator1, Loop Behavior=::NewEnumerator1, Loop Delay=0.07, Loop Duration=0.5, UseLoopDelay=true)
   &SpawnBurst_Instantaneous(Use Spawn Probability=false, UseLoopCountLimit=false)
  ]
  particleSpawnScript [
   &InitializeParticle(Color=<rgba 0.6445 0.6445 0.6445 0.5>, Lifetime=0.25, Mesh Scale Mode=::NewEnumerator4, UsePositionOffset=false, Write Material Random=true)
   &AddVelocity(Velocity=<0 0 -100>)
   &InitialMeshOrientation(Mesh Orientation Mode=::NewEnumerator2, Rotation=<0 -0.25 0>, Use Rotation Vector=true)
   &SkeletalMeshLocation(
    Bone Sampling Mode=::NewEnumerator4
    MaskSpawnGroup=false
    Orientation Sampling=::NewEnumerator1
    Sampled Position Offset=<0 0 100>
    Skeletal Mesh=$User.Skeletal Mesh
   )
   &CameraOffset(Camera Offset Amount=30)
  ]
  particleUpdateScript [
   &ParticleState()
   &ScaleMeshSize(Scale Factor=&MakeVector(
     X=&FloatFromCurve(FloatCurve=DataInterface() {propertyValues {Curve {keys [<at 0 1>,<at 1 1.5>]}}})
     Y=1
     Z=1
    ))
   &SolveForcesAndVelocity(Clamp Velocity=false, Limit Acceleration=false)
   &DynamicMaterialParameters(
    Index 0 Param 1=&FloatFromCurve(FloatCurve=DataInterface() {propertyValues {Curve {keys [<at 0 0>,<at 0.5 0>,<at 1 0.08>]}}})
    Index 0 Param 2=&FloatFromCurve(FloatCurve=DataInterface() {propertyValues {Curve {keys [{arriveTangent=-2.2616, interpMode=RCIM_Cubic, time=0, value=-0.5}, {interpMode=RCIM_Cubic, leaveTangent=-0.7156, time=1, value=1}]}}})
    Index 0 Param 3=1.2
    Index 0 Param 4=&FloatFromCurve(FloatCurve=DataInterface() {propertyValues {Curve {keys [<at 0 1>,<at 1 2>]}}}, Scale Curve=0.2)
    Module.Param0WriteEnabled=true
    Module.Param1WriteEnabled=true
    Module.Param2WriteEnabled=true
    Module.Param3WriteEnabled=true
   )
   &ScaleColor(
    Scale Alpha=&FloatFromCurve(FloatCurve=DataInterface() {propertyValues {Curve {keys [<at 0 1 RCIM_Cubic>, {tangentMode=RCTM_User, time=1, value=0}]}}})
    ScaleA=true
    ScaleRGB=false
   )
  ]
  rendererClasses [&NiagaraMeshRendererProperties]
  renderers [
   {rendererClass=&NiagaraMeshRendererProperties, rendererIndex=0}
   {rendererClass=&NiagaraMeshRendererProperties, rendererIndex=1}
  ]
 }
 Droplets {
  bEnabled=false
  emitterUpdateScript [
   &EmitterState()
   &SpawnBurst_Instantaneous(
    Spawn Count=&RandomRangeInt(Minimum=3, Maximum=6)
    Spawn Time=0.05
    Use Spawn Probability=false
    UseLoopCountLimit=false
   )
  ]
  particleSpawnScript [
   &InitializeParticle(
    Color Maximum=<rgba 1.4167 1.7745 2 0.9435>
    Color Minimum=<rgba 1.4167 1.7745 2 0.5>
    Color Mode=::NewEnumerator2
    Lifetime=&UniformRangedFloat(Minimum=0.2, Maximum=0.6)
    Mass=&UniformRangedFloat(Minimum=0.6, Maximum=1.2)
    Mass Mode=::NewEnumerator1
    Mesh Scale Mode=::NewEnumerator1
    Mesh Uniform Scale Max=0.1
    Mesh Uniform Scale Min=0.06
    Sprite Size=&Vector2DFromCurve(
     Vector2Curve=DataInterface() {propertyValues {
       XCurve {keys [<at 0 2 RCIM_Cubic>,<at 1 1.25 RCIM_Cubic>]}
       YCurve {keys [
         {arriveTangent=-2.6137, interpMode=RCIM_Cubic, leaveTangent=-2.6137, leaveTangentWeight=0.5503, tangentMode=RCTM_User, time=0.0044, value=5}
         {arriveTangent=-4.0179, interpMode=RCIM_Cubic, leaveTangent=-4.0179, time=0.2382, value=4}
         <at 1 1 RCIM_Cubic>
        ]}
      }}
     CurveIndex=&RandomRangeFloat(Minimum=0, Maximum=1)
     ScaleCurve=&Vector2DFromFloat(Value=&RandomRangeFloat(Minimum=0.9, Maximum=1.25))
    )
    Sprite Size Mode=::NewEnumerator4
    UsePositionOffset=false
    Write Material Random=true
   )
   &ShapeLocation(
    Apply To Particle Position=::NewEnumerator1
    MaskSpawnGroup=false
    Override Seed=false
    Rotation Axis=<0 0 1>
    Rotation Mode=::NewEnumerator1
    Shape Origin=&VectorFromFloat(Value=0)
    Sphere Radius=3
    UseOwnerScale=true
   )
   &AddVelocity(
    Module.Use Falloff Distance=false
    Override Seed=false
    Velocity Mode=::NewEnumerator1
    Velocity Origin=$Particles.ShapeLocation.ShapePosition
    Velocity Speed=&FloatFromCurve(
     FloatCurve=DataInterface() {propertyValues {Curve {keys [<at 0 150>,<at 1 50>]}}}
     CurveIndex=&ReturnNormalizedExecIndex()
    )
   ) {module=AddVelocity001}
   &InheritVelocity(
    Inherited Velocity Amount Scale=&VectorFromFloat(Value=-0.15)
    Inherited Velocity Speed Limit:NiagaraFloat=&RandomRangeFloat(Minimum=100, Maximum=300)
    Speed Limit=true
    UseSpeedThreshold=false
   )
   &SkeletalMeshLocation(
    Bone / Socket Index=3
    MaskSpawnGroup=false
    Mesh Sampling Type=::NewEnumerator5
    Orientation Sampling=::NewEnumerator1
    Position Sampling=::NewEnumerator1
    Sampled Position Offset:NiagaraPosition=$Particles.ShapeLocation.ShapePosition
    Skeletal Mesh=$User.Skeletal Mesh
    Socket Sampling Mode=::NewEnumerator5
   ) {module=SkeletalMeshLocation001}
  ]
  particleUpdateScript [
   &ParticleState()
   &SkeletalMeshLocation(
    Bone / Socket Index=3
    MaskSpawnGroup=false
    Mesh Sampling Type=::NewEnumerator5
    Orientation Sampling=::NewEnumerator1
    Position Sampling=::NewEnumerator1
    Sampled Position Offset:NiagaraPosition=$Particles.ShapeLocation.ShapePosition
    Skeletal Mesh=$User.Skeletal Mesh
    Socket Sampling Mode=::NewEnumerator5
   )
   &ScaleMeshSize(Scale Factor=&VectorFromFloat(Value=&FloatFromCurve(FloatCurve=DataInterface() {propertyValues {Curve {keys [
          {arriveTangent=-2.0458, interpMode=RCIM_Cubic, time=0.5, value=1}
          {arriveTangent=-4.0006, interpMode=RCIM_Cubic, leaveTangent=-4.0006, tangentMode=RCTM_User, time=1, value=0}
         ]}}})))
   &ScaleSpriteSize(
    Scale Factor=&Multiply_Vector2D_ByFloat(
     Float=&FloatFromCurve(FloatCurve=DataInterface() {propertyValues {Curve {
         defaultValue=1
         keys [
          {arriveTangent=19.9371, interpMode=RCIM_Cubic, leaveTangent=19.9371, tangentMode=RCTM_User, time=0, value=0}
          {arriveTangentWeight=0.0905, interpMode=RCIM_Cubic, leaveTangentWeight=0.8147, tangentMode=RCTM_User, time=0.1, value=0.9847}
          {arriveTangentWeight=0.1976, interpMode=RCIM_Cubic, leaveTangentWeight=0.247, tangentMode=RCTM_User, time=0.5, value=1}
          {arriveTangent=-2.384, interpMode=RCIM_Cubic, leaveTangent=-2.384, tangentMode=RCTM_User, time=1, value=0}
         ]
        }}})
     Vector2D=&Vector2DFromCurve(
      Vector2Curve=DataInterface() {propertyValues {
        XCurve {keys [{tangentMode=RCTM_User, time=0, value=1}, {tangentMode=RCTM_User, time=1, value=1}]}
        YCurve {keys [
          {interpMode=RCIM_Cubic, tangentMode=RCTM_User, time=0, value=1.6317}
          {arriveTangent=-2.6411, interpMode=RCIM_Cubic, leaveTangent=-2.6411, tangentMode=RCTM_User, time=0.3595, value=1}
          {tangentMode=RCTM_User, time=1, value=0.2}
         ]}
       }}
      ScaleCurve=<1 1>
     )
    )
    Scale Sprite Size Mode=::NewEnumerator2
   )
   &ScaleColor(
    Scale Alpha=&FloatFromCurve(FloatCurve=DataInterface() {propertyValues {Curve {keys [
         <at 0 0 RCIM_Cubic>
         {arriveTangent=0.002, interpMode=RCIM_Cubic, leaveTangent=0.002, time=0.1183, value=1.0113}
         <at 1.0011 0.002 RCIM_Cubic>
        ]}}})
    ScaleA=true
    ScaleRGB=true
   )
   &Drag(Drag=&RandomRangeFloat(Minimum=1, Maximum=3), Use Linear Drag=true, Use Rotational Drag=false)
   &GravityForce()
   &SolveForcesAndVelocity(Clamp Velocity=false, Limit Acceleration=false)
  ]
  rendererClasses [&NiagaraSpriteRendererProperties]
  renderers [{rendererClass=&NiagaraSpriteRendererProperties}]
 }
}
summary {
 systemName=NS_Pickaxe_MicStand_BlobRock_FirstAttack
 emitters [
  {emitterName=Airflow_Stick, rendererClasses [&NiagaraMeshRendererProperties]}
  {emitterName=Air_cone, rendererClasses [&NiagaraMeshRendererProperties]}
  {bEnabled=false, emitterName=Droplets, rendererClasses [&NiagaraSpriteRendererProperties]}
 ]
 userVariables [User.Skeletal Mesh {
   type=NiagaraDataInterfaceSkeletalMesh
   defaultValue {
    struct=NiagaraExt_VariableValue_DataInterface
    value {dataInterface=&self:NiagaraDataInterfaceSkeletalMesh_0}
   }
  }]
}
